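# Landscape material function for Creative Asteria that adds height-based gradient color variation to cliff base color.
# Axis-weighted world position, summing X twice plus Z without Y, is scaled and offset by roughness-driven distortion.
# That coordinate is rotated by a band angle to sample a cliff gradient texture.
# The biased gradient multiplies base color, blended by an amount parameter; older CliffRock parameters sit unconnected.
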MaterialFunction /Game/Creative/Materials/Landscape/Creative_Landscape_Asteria/MaterialFunctions/MF_CR_CliffZColor
graph {nodes [
  Constant3Vector() {constant=<rgba -0.1 -0.1 1 1>}
  Constant()
  Constant() {r=0.1}
  Constant() {r=10000}
  Constant() {r=0.2}
  Constant() {r=0.284}
  n1 = GetMaterialAttributes(None=FunctionInput() {inputType=FunctionInput_MaterialAttributes}) {attributeGetTypes [#1, #2, #3]}
  n2 = ScalarParameter() {group=Cliff, parameterName=Cliff_ZGradient_BandAngle, sortPriority=2400}
  n3 = ScalarParameter() {
   defaultValue=0.1
   desc=Uses the Cliff Roughness texture to distort the ZGradient texture. Set this to zero for no distortion.
   group=Cliff
   parameterName=Cliff_ZGradient_DistortUsingRoughness
   sortPriority=2500
  }
  n4 = LinearInterpolate(A=Multiply(A=@n3) {constB=-1}, B=@n3, Alpha=@n1.Roughness)
  n5 = VectorParameter() {group=Cliff, parameterName=Cliff_ZGradient_AxisIntensity, sortPriority=2300, defaultValue=<rgba -0.1 -0.1 1 1>}
  n6 = MaterialFunctionCall(Float3=Multiply(A=WorldPosition().XYZ, B=@n5.RGB)) {materialFunction=&BreakOutFloat3Components}
  n7 = Divide(
   A=Add(A=Add(A=@n6.R, B=@n6.R), B=@n6.B)
   B=ScalarParameter() {defaultValue=10000, group=Cliff, parameterName=Cliff_ZGradient_Scale, sortPriority=2100}
  )
  n8 = TextureSample(UVs=MaterialFunctionCall!(
    Angle=MaterialFunctionCall!(Degrees=@n2) {materialFunction=/Engine/Functions/Engine_MaterialFunctions01/Math/DegreesToRadians}.Radians
    UV/2DVector=Add(A=@n4, B=@n7)
   ) {materialFunction=/Game/Characters/Player/Male/Medium/Bodies/M_MED_Faux/Materials/MF_2DRotate}.Result) {samplerSource=SSM_Wrap_WorldGroupSettings, texture=/Game/Athena/Asteria/Environments/Landscape/Materials/MaterialFunctions/T_Cliff_Grad}
  n9 = ScalarParameter() {defaultValue=0.2, group=Cliff, parameterName=Cliff_ZGradient_AddBias, sortPriority=2200}
  n10 = Saturate(None=Multiply(A=@n1.BaseColor, B=Add(A=@n8.RGB, B=@n9)))
  n11 = ScalarParameter() {
   defaultValue=0.284
   desc=Uses a gradient texture to give color variation to the cliffs. Set this to zero for no color variation.
   group=Cliff
   parameterName=Cliff_ZGradient_Amount
   sortPriority=2000
  }
  n12 = Saturate(None=@n11)
  FunctionOutput(None=SetMaterialAttributes!(MaterialAttributes=@n1.MaterialAttributes, BaseColor=LinearInterpolate(A=@n1.BaseColor, B=@n10, Alpha=@n12)) {attributeSetTypes [#1]})
  Multiply(A=@n1.Normal, B=@n3)
  ScalarParameter() {group=Cliff, parameterName=CliffRockZBandAngle}
  ScalarParameter() {defaultValue=0.5, group=Cliff, parameterName=CliffRockZGradOpacity}
  ScalarParameter() {defaultValue=0.3, group=Cliff, parameterName=CliffRockZGradRoughnessMod}
  ScalarParameter() {defaultValue=10000, group=Cliff, parameterName=CliffRockZGradSize}
  ScalarParameter() {defaultValue=0.2, group=Cliff, parameterName=CliffRockZGradPreBias}
  VectorParameter() {group=Cliff, parameterName=CliffRockZGradAxisIntensity, defaultValue=<rgba -0.1 -0.1 1 1>}
 ]}
