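# Material function that builds the cliff layer for the Creative Asteria landscape, outputting material, height, cliff-face height and emissive mask.
# Two parameterized world-aligned rock sets are height-blended using a large-scale swirl-noise phase.
# The default shading path adds rock-detail modulation to base color and specular, then applies MF_CR_CliffZColor, while mobile takes a separate input.
# Slope, noise and height lerps shape edge masking; Displacement and SlopeScreeMask output constant zero.
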MaterialFunction /Game/Creative/Materials/Landscape/Creative_Landscape_Asteria/MaterialFunctions/MF_CR_AsteriaCommonCliff
graph {nodes [
  n1 = TextureObjectParameter() {
   group=Cliff
   parameterName=Cliff_TextureSet2_Basecolor
   sortPriority=-499
   texture=/Game/Environments/Asteria/Terrain/Textures/T_Asteria_RockCliff_03_D
  }
  n2 = TextureObjectParameter() {
   group=Cliff
   parameterName=Cliff_TextureSet2_SpecularRoughness
   samplerType=SAMPLERTYPE_Masks
   sortPriority=-400
   texture=/Game/Environments/Asteria/Terrain/Textures/T_Asteria_RockCliff_03_S
  }
  n3 = VectorParameter() {group=Cliff, parameterName=Cliff_TextureSet2_Scale, sortPriority=-500, defaultValue=<rgba 3000 3000 3000 1>}
  n4 = TextureObjectParameter() {
   group=Cliff
   parameterName=Cliff_TextureSet2_NormalMap
   samplerType=SAMPLERTYPE_Normal
   sortPriority=-300
   texture=/Game/Environments/Asteria/Terrain/Textures/T_Asteria_RockCliff_03_N
  }
  n5 = MaterialFunctionCall!(
   Diffuse Texture Object=@n1
   Spec Texture Object=@n2
   TextureSize=@n3.RGB
   --------------------------
   Use seperate settings for Z projection
   Use seperate texture for Z projection
   Texture Size Z projection
   Z Plane Diffuse Texture Object
   Z Plane Spec Texture Object
   "--------------------------      "
   ProjectionTransitionContrast=Constant() {r=0.5}
   Use Texture Alpha Heightmap to Refine the Transistion=StaticBool() {value=true}
   HeightLerpTransitionContrast=Constant() {r=0.5}
   PickTheDiffuseHeightChannel
   TheDiffuseHeightChannel
   "--------------------------       "
   NormalTextureObject=@n4
   World Space Normals=StaticBool()
   "--------------------------  "
   Export Float 4 Diffuse=StaticBool() {value=true}
   Export Float 4 Spec
   "--------------------------     "
   WorldPosition=WorldPosition().XYZ
   World Space Normal
  ) {materialFunction=/Engine/Functions/Engine_MaterialFunctions02/Texturing/WorldAlignedTextures_Complex}
  ComponentMask(None=@n5.XY Spec) {b=true}
  ComponentMask(None=@n5.XY Spec) {r=true}
  Constant3Vector() {constant=<rgba 4000 4000 4000 1>}
  Constant3Vector() {constant=<rgba 3000 3000 3000 1>}
  n6 = Constant() {r=2.05}
  n7 = MaterialFunctionCall(
   Diffuse Texture Object=TextureObject() {texture=/Game/Environments/Apollo/Rocks/Natural_Rocks/Textures/T_RockDetails_02D}
   TextureSize=Constant() {r=700}
   ProjectionTransitionContrast=Constant() {r=5}
  ) {materialFunction=&WorldAlignedTextures_Complex}
  n8 = ComponentMask(None=@n7.XY Diffuse Texture) {r=true}
  n9 = Saturate(None=Add(A=Constant() {r=-1.1}, B=Multiply(A=@n6, B=@n8)))
  n10 = Saturate(None=LinearInterpolate(A=Constant() {r=7}, B=Constant() {r=-0.8}, Alpha=@n8))
  n11 = GetMaterialAttributes(None=NamedRerouteUsage()) {attributeGetTypes [
    #1
    #2
    #3
    #4
   ]}
  n12 = Saturate(None=LinearInterpolate(A=Constant() {r=-0.25}, B=Constant() {r=1.5}, Alpha=@n11.Specular))
  n13 = Constant() {r=0.5}
  n14 = Multiply(A=@n11.BaseColor, B=SquareRoot(None=@n7.XY Diffuse Texture))
  n15 = Constant() {r=1.5}
  n16 = SetMaterialAttributes!(
   MaterialAttributes=@n11.MaterialAttributes
   BaseColor=Saturate(None=Multiply(A=@n14, B=@n15))
   Specular=Multiply(A=Multiply(A=@n9, B=@n10), B=Clamp(None=@n12, Max=@n13) {clampMode=CMODE_ClampMax})
  ) {attributeSetTypes [#4, #3]}
  FunctionOutput(None=ShadingPathSwitch(
    Default=MaterialFunctionCall!(In=@n16) {materialFunction=&/MF_CR_CliffZColor}.Result
    Mobile=NamedRerouteUsage()
   )) {outputName=Material, sortPriority=-1}
  FunctionOutput(None=Constant()) {outputName=Displacement, sortPriority=20}
  FunctionOutput(None=Constant()) {outputName=SlopeScreeMask, sortPriority=10}
  n17 = NamedRerouteUsage()
  n18 = Saturate(None=Multiply(A=OneMinus(None=@n17), B=Constant() {r=3}))
  n19 = MaterialFunctionCall(Transition Phase=@n18, Height Texture=ShadingPathSwitch(Default=NamedRerouteUsage(), Mobile=NamedRerouteUsage())) {materialFunction=&HeightLerp}
  FunctionOutput(None=NamedRerouteDeclaration(None=Max(A=@n18, B=@n19.Lerp Alpha No Contrast) {desc="For BLEND MATERIAL, smallrockedge-mat should be in B"}) {name=OutputHeight, nodeColor=<rgba 1 0 1 1>}) {outputName=Height}
  n20 = TextureSample(UVs=MaterialFunctionCall!(Scale=Constant() {r=-10000}) {materialFunction=/Engine/Functions/Engine_MaterialFunctions01/Coordinates/WorldPosition-XY}.Result) {
   samplerSource=SSM_Wrap_WorldGroupSettings
   samplerType=SAMPLERTYPE_LinearGrayscale
   texture=/Game/Global/Textures/Default/Noise/T_SwirlNoise_N_height
  }
  n21 = MaterialFunctionCall!(
   Diffuse 1
   Diffuse 2
   Transistion Phase=@n20.R
   Height Texture 1=NamedRerouteUsage()
   Height Texture 2=NamedRerouteUsage()
   Contrast=Constant() {desc="Play with neg or pos ##", r=-5}
  ) {materialFunction=/Engine/Functions/Engine_MaterialFunctions02/Texturing/HeightLerpWithTwoHeightMaps}
  FunctionOutput(None=@n21.Resulting Height Map) {outputName=CliffFaceHeightMap}
  n22 = NamedRerouteUsage()
  n23 = NamedRerouteUsage()
  FunctionOutput(None=NamedRerouteDeclaration(None=LinearInterpolate(A=@n22, B=@n23, Alpha=@n21.Alpha)) {name=EmissiveMask_Result, nodeColor=<rgba 0 1 0.3906 1>}) {outputName=CliffEmissiveMask}
  n24 = BlendMaterialAttributes(A=NamedRerouteUsage(), B=NamedRerouteUsage(), Alpha=@n21.Alpha)
  NamedRerouteDeclaration(None=@n24) {name=BlendedMaterialBeforeDetailTexture, nodeColor=<rgba 1 0 1 1>}
  NamedRerouteDeclaration(None=ComponentMask(None=@n5.XY Spec) {g=true}) {name=MediumRock_EmissiveMask, nodeColor=<rgba 0 0.3843 1 1>}
  n25 = ChannelMaskParameter(None=@n5.XY Spec) {
   desc="Define the texture channel used for the Specular mask.\r\nDefault is red."
   group=Cliff
   parameterName=Cliff_TextureSet2_SpecularChannel
   sortPriority=-399
  }
  n26 = ChannelMaskParameter(None=@n5.XY Spec) {
   desc="Define the texture channel used for the Roughness mask.\r\nMany of the older Fortnite textures use green channel for metallic and blue channel for roughness."
   group=Cliff
   maskChannel=Blue
   parameterName=Cliff_TextureSet2_RoughnessChannel
   sortPriority=-398
   defaultValue=<rgba 0 0 1 0>
  }
  n27 = SetMaterialAttributes!(
   MaterialAttributes
   BaseColor=ComponentMask(None=@n5.XY Diffuse Texture) {b=true, g=true, r=true}
   Specular=@n25.R
   Roughness=@n26.R
   Normal=@n5.Normal XY Texture
  ) {attributeSetTypes [
    #4
    #3
    #2
    #1
   ]}
  NamedRerouteDeclaration(None=@n27) {name=MediumRock-Material, nodeColor=<rgba 1 0 1 1>}
  n28 = TextureObjectParameter() {
   group=Cliff
   parameterName=Cliff_TextureSet1_Basecolor
   sortPriority=-1000
   texture=/Game/Environments/Asteria/Terrain/Textures/T_Asteria_RockCliff_04_D
  }
  n29 = TextureObjectParameter() {
   group=Cliff
   parameterName=Cliff_TextureSet1_SpecularRoughness
   samplerType=SAMPLERTYPE_Masks
   sortPriority=-900
   texture=/Game/Environments/Asteria/Terrain/Textures/T_Asteria_RockCliff_04_S
  }
  n30 = VectorParameter() {group=Cliff, parameterName=Cliff_TextureSet1_Scale, sortPriority=-1100, defaultValue=<rgba 4000 4000 4000 1>}
  n31 = TextureObjectParameter() {
   group=Cliff
   parameterName=Cliff_TextureSet1_NormalMap
   samplerType=SAMPLERTYPE_Normal
   sortPriority=-800
   texture=/Game/Environments/Asteria/Terrain/Textures/T_Asteria_RockCliff_04_N
  }
  n32 = MaterialFunctionCall(
   Diffuse Texture Object=@n28
   Spec Texture Object=@n29
   TextureSize=@n30.RGB
   ProjectionTransitionContrast=Constant() {r=0.5}
   Use Texture Alpha Heightmap to Refine the Transistion=StaticBool() {value=true}
   HeightLerpTransitionContrast=Constant() {r=0.5}
   NormalTextureObject=@n31
   World Space Normals=ShadingPathSwitch(Default=StaticBool(), Mobile=StaticBool() {value=true})
   Export Float 4 Diffuse=StaticBool() {value=true}
   WorldPosition=WorldPosition().XYZ
  ) {materialFunction=&WorldAlignedTextures_Complex}
  n33 = ComponentMask(None=@n32.XY Diffuse Texture) {b=true, g=true, r=true}
  n34 = ChannelMaskParameter(None=@n32.XY Spec) {
   desc="Define the texture channel used for the Specular mask.\r\nDefault is red."
   group=Cliff
   parameterName=Cliff_TextureSet1_SpecularChannel
   sortPriority=-899
  }
  n35 = ChannelMaskParameter(None=@n32.XY Spec) {
   desc="Define the texture channel used for the Roughness mask.\r\nMany of the older Fortnite textures use green channel for metallic and blue channel for roughness."
   group=Cliff
   maskChannel=Blue
   parameterName=Cliff_TextureSet1_RoughnessChannel
   sortPriority=-898
   defaultValue=<rgba 0 0 1 0>
  }
  n36 = SetMaterialAttributes(BaseColor=@n33, Specular=@n34.R, Roughness=@n35.R, Normal=@n32.Normal XY Texture) {attributeSetTypes [
    #4
    #3
    #2
    #1
   ]}
  NamedRerouteDeclaration(None=@n36) {name=LargeRock-Material, nodeColor=<rgba 1 0 1 1>}
  NamedRerouteDeclaration(None=ComponentMask(None=@n32.XY Diffuse Texture) {a=true}) {name=LargeRock-Height, nodeColor=<rgba 1 0 1 1>}
  NamedRerouteDeclaration(None=ComponentMask(None=@n5.XY Diffuse Texture) {a=true}) {name=MediumRock-Height, nodeColor=<rgba 1 0 1 1>}
  NamedRerouteDeclaration(None=ComponentMask(None=@n32.XY Spec) {g=true}) {name=LargeRock_EmissiveMask, nodeColor=<rgba 0 1 0.8594 1>}
  n37 = ComponentMask(None=VertexNormalWS()) {b=true}
  NamedRerouteDeclaration(None=FunctionInput(Preview=Abs(None=@n37)) {bUsePreviewValueAsDefault=true, inputName=Slope, inputType=FunctionInput_Scalar, sortPriority=4}) {name=BasicSlope, nodeColor=<rgba 1 0 1 1>}
  NamedRerouteDeclaration(None=Constant()) {
   name=SmallRockEdge-UnmodifiedHeight
   desc="This was previously using the alpha channel of T_Apollo_Dirt_D_02\r\nw/o any modifiers.. There are some very slight, but acceptable\r\ndifferences attributed to T_Apollo_Dirt_S_02 not being in sRGB space."
   nodeColor=<rgba 1 0 1 1>
  }
  n38 = TextureSample(UVs=MaterialFunctionCall(Scale=Constant() {r=1200}) {materialFunction=&WorldPosition-XY}.Result) {
   samplerSource=SSM_Wrap_WorldGroupSettings
   samplerType=SAMPLERTYPE_LinearGrayscale
   texture=&T_SwirlNoise_N_height
  }
  NamedRerouteDeclaration(None=OneMinus(None=@n38.A)) {
   name=SmallRockEdge-Height
   desc="This was previously using the alpha channel of T_Apollo_Dirt_D_02\r\nw/o any modifiers.. There are some very slight, but acceptable\r\ndifferences attributed to T_Apollo_Dirt_S_02 not being in sRGB space."
   nodeColor=<rgba 1 0 1 1>
  }
  n39 = Multiply(A=ComponentMask(None=@n33) {r=true}, B=Constant() {r=2.25})
  n40 = Constant() {r=-0.3}
  n41 = SetMaterialAttributes(BaseColor=@n33, Specular=Saturate(None=Add(A=@n39, B=@n40)), Roughness=Constant() {r=0.85}, Normal=@n32.Normal XY Texture) {attributeSetTypes [
    #4
    #3
    #2
    #1
   ]}
  NamedRerouteDeclaration(None=@n41) {name=Mobile-Material, nodeColor=<rgba 1 0 1 1>}
  n42 = Saturate(None=LinearInterpolate(A=Constant() {r=-6}, B=Constant() {r=2.5}, Alpha=NamedRerouteUsage()) {desc=Adjust the outer edge of the small rock})
  n43 = Constant() {r=5}
  n44 = Constant() {r=-2}
  n45 = NamedRerouteUsage()
  n46 = Add(A=Power(Base=NamedRerouteUsage(), Exp=Constant() {r=3}), B=Saturate(None=LinearInterpolate(A=@n43, B=@n44, Alpha=@n45)))
  n47 = Multiply(A=Power(Base=NamedRerouteUsage(), Exp=Constant() {r=5}), B=Constant() {r=3})
  n48 = OneMinus(None=MaterialFunctionCall(Transition Phase=Saturate(None=@n46), Height Texture=Saturate(None=@n47)) {materialFunction=&HeightLerp}.Lerp Alpha No Contrast)
  n49 = Constant() {r=-1}
  n50 = Constant() {r=1.15}
  n51 = NamedRerouteUsage()
  n52 = Multiply(A=Saturate(None=Add(A=@n42, B=@n48)), B=Saturate(None=LinearInterpolate(A=@n49, B=@n50, Alpha=@n51)))
  n53 = NamedRerouteUsage()
  n54 = Constant() {r=1}
  NamedRerouteDeclaration(None=Multiply(
    A=MaterialFunctionCall(Transition Phase=@n52, Height Texture=@n53, Contrast=@n54) {materialFunction=&HeightLerp}.Alpha
    B=Constant() {r=0.75}
   )) {name=SmallRockEdgeMask, nodeColor=<rgba 1 0 1 1>}
  NamedRerouteDeclaration(None=GetMaterialAttributes(None=@n24) {attributeGetTypes [#1]}.Normal) {name=MediumAndLargeRockNormals, nodeColor=<rgba 1 0 1 1>}
  ScalarParameter() {defaultValue=3, group=Cliff, parameterName=6y4et87y3rtugg}
  StaticSwitch(True=@n16, False=NamedRerouteUsage(), Value=StaticBool() {value=true})
  TextureObject() {texture=&T_Asteria_RockCliff_04_D}
  TextureObject() {samplerType=SAMPLERTYPE_Masks, texture=&T_Asteria_RockCliff_04_S}
  TextureObject() {samplerType=SAMPLERTYPE_Normal, texture=&T_Asteria_RockCliff_04_N}
  TextureObject() {texture=&T_Asteria_RockCliff_03_D}
  TextureObject() {samplerType=SAMPLERTYPE_Masks, texture=&T_Asteria_RockCliff_03_S}
  TextureObject() {samplerType=SAMPLERTYPE_Normal, texture=&T_Asteria_RockCliff_03_N}
 ]}
